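# Editor material MAT_EnableTrack_Mat, filed under Matinee editor materials, apparently for displaying an enable indicator on Matinee tracks.
# It samples the Cascade CASC_ModuleEnable texture and feeds its RGB as DiffuseColor into the ConvertFromDiffSpec material function.
# The material uses the Default Lit shading model and legacy pixel depth offset mode.
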Material /Engine/EditorMaterials/Matinee/MAT_EnableTrack_Mat
graph {
 properties {pixelDepthOffsetMode=PDOM_Legacy, usedShadingModels=Default Lit}
 nodes [MaterialFunctionCall(DiffuseColor=TextureSample() {texture=/Engine/EditorMaterials/Cascade/CASC_ModuleEnable}.RGB) {materialFunction=&ConvertFromDiffSpec}]
}
